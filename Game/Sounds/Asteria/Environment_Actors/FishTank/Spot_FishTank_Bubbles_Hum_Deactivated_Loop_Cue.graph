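# Looping indoor ambient cue for a deactivated fish tank's bubble hum, with spatial attenuation, occlusion settings and a reverb send.
# Two nested looping envelopers add slow pitch and volume drift to the hum loop.
# A looping random delay plays one of ten borrowed distant blimp air-release one-shots, mixed at half volume under the hum.
# The first two quality slots get this mix, while the last two play only the hum.
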SoundCue /Game/Sounds/Asteria/Environment_Actors/FishTank/Spot_FishTank_Bubbles_Hum_Deactivated_Loop_Cue
root=QualityLevel_0
nodes [
 {id=Delay_2, children [Random_2], properties {delayMax=12, delayMin=3}}
 {
  id=Enveloper_0
  children [WavePlayer_1]
  properties {
   bLoop=true
   bLoopIndefinitely=true
   loopEnd=30
   pitchCurve {editorCurveData {keys [
      {arriveTangent=-0.0013, interpMode=RCIM_Cubic, leaveTangent=-0.0013, time=0}
      <at 15 0.98 RCIM_Cubic>
      {arriveTangent=0.0013, interpMode=RCIM_Cubic, leaveTangent=0.0013, time=30}
     ]}}
   volumeCurve {editorCurveData {keys [<at 0 1>]}}
  }
 }
 {
  id=Enveloper_3
  children [Enveloper_0]
  properties {
   bLoop=true
   bLoopIndefinitely=true
   loopEnd=10
   pitchCurve {editorCurveData {keys [{time=0}]}}
   volumeCurve {editorCurveData {keys [
      {arriveTangent=-0.01, interpMode=RCIM_Cubic, time=0, value=1}
      <at 5 0.7 RCIM_Cubic>
      {arriveTangent=-0.03, interpMode=RCIM_Cubic, leaveTangent=-0.03, time=10, value=1}
      <at 15 0.4 RCIM_Cubic>
      <at 20 1 RCIM_Cubic>
     ]}}
  }
 }
 {id=Looping_2, children [Delay_2]}
 {id=Mixer_1, children [Looping_2,Enveloper_3], properties {inputVolume [0.5,1]}}
 {id=QualityLevel_0, children [Mixer_1,Mixer_1,Enveloper_3,Enveloper_3]}
 {
  id=Random_2
  children [
   WavePlayer_12
   WavePlayer_13
   WavePlayer_14
   WavePlayer_15
   WavePlayer_16
   WavePlayer_17
   WavePlayer_18
   WavePlayer_19
   WavePlayer_20
   WavePlayer_21
  ]
  properties {weights [1,1,1,1,1,1,1,1,1,1]}
 }
 {
  id=WavePlayer_1
  properties {bLooping=true, soundWaveAssetPtr=&/Spot_FishTank_Bubble_Hum_Deactivated_Loop}
 }
 {id=WavePlayer_12, properties {soundWaveAssetPtr=/Game/Sounds/Artemis/POI/Blimps/Spots/Spot_Blimp_Air_Release_Distant_01_Stereo}}
 {id=WavePlayer_13, properties {soundWaveAssetPtr=/Game/Sounds/Artemis/POI/Blimps/Spots/Spot_Blimp_Air_Release_Distant_04_Stereo}}
 {id=WavePlayer_14, properties {soundWaveAssetPtr=/Game/Sounds/Artemis/POI/Blimps/Spots/Spot_Blimp_Air_Release_Distant_10_Stereo}}
 {id=WavePlayer_15, properties {soundWaveAssetPtr=/Game/Sounds/Artemis/POI/Blimps/Spots/Spot_Blimp_Air_Release_Distant_02_Stereo}}
 {id=WavePlayer_16, properties {soundWaveAssetPtr=/Game/Sounds/Artemis/POI/Blimps/Spots/Spot_Blimp_Air_Release_Distant_03_Stereo}}
 {id=WavePlayer_17, properties {soundWaveAssetPtr=/Game/Sounds/Artemis/POI/Blimps/Spots/Spot_Blimp_Air_Release_Distant_05_Stereo}}
 {id=WavePlayer_18, properties {soundWaveAssetPtr=/Game/Sounds/Artemis/POI/Blimps/Spots/Spot_Blimp_Air_Release_Distant_06_Stereo}}
 {id=WavePlayer_19, properties {soundWaveAssetPtr=/Game/Sounds/Artemis/POI/Blimps/Spots/Spot_Blimp_Air_Release_Distant_07_Stereo}}
 {id=WavePlayer_20, properties {soundWaveAssetPtr=/Game/Sounds/Artemis/POI/Blimps/Spots/Spot_Blimp_Air_Release_Distant_08_Stereo}}
 {id=WavePlayer_21, properties {soundWaveAssetPtr=/Game/Sounds/Artemis/POI/Blimps/Spots/Spot_Blimp_Air_Release_Distant_09_Stereo}}
]
properties {
 bOverrideAttenuation=true
 bOverrideConcurrency=true
 maxDistance=920
 soundClassObject=&Fort_Ambient_Indoors_Athena
 subtitlePriority=5000
 volumeMultiplier=0.501
 attenuationOverrides {
  bAttenuateWithLPF=true
  bEnableFocusInterpolation=true
  bEnableListenerFocus=true
  dBAttenuationAtMax=-40
  distanceAlgorithm=Logarithmic
  falloffDistance=750
  focusAzimuth=15
  lPFFrequencyAtMax=1000
  lPFRadiusMax=1000
  lPFRadiusMin=500
  nonFocusAzimuth=45
  nonFocusVolumeAttenuation=0.5
  occlusionInterpolationTime=2
  occlusionLowPassFilterFrequency=3000
  occlusionTraceChannel=ECC_GameTraceChannel11
  occlusionVolumeAttenuation=0.2
  stereoSpread=100
  attenuationShapeExtents=<170 173.2973 131.6805>
 }
 concurrencyOverrides {bLimitToOwner=true, maxCount=2}
 soundSubmixSends [{
   minSendLevel=0.2
   soundSubmix=&Fortnite_Reverb_Default
   customSendLevelCurve {editorCurveData {keys [<at 0 0.2>,<at 1 1>]}}
  }]
}
