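# Looping ambient sound cue for a fish tank spot emitter, combining a bubbling hum with intermittent air-rush and swell one-shots.
# Two looping Delay-into-Random branches are mixed with a looping hum, whose nested Envelopers slowly modulate pitch and volume.
# The first two quality levels play the full mix; the last two play only the enveloped hum.
# Overrides add capsule attenuation with low-pass filtering and listener focus, plus per-owner concurrency.
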SoundCue /Game/Sounds/Asteria/Environment_Actors/FishTank/Spot_FishTank_Bubbles_Hum_Loop_Cue
root=QualityLevel_0
nodes [
 {id=Delay_0, children [Random_0], properties {delayMax=12, delayMin=4}}
 {id=Delay_1, children [Random_1], properties {delayMax=8, delayMin=3}}
 {
  id=Enveloper_0
  children [WavePlayer_0]
  properties {
   bLoop=true
   bLoopIndefinitely=true
   loopEnd=30
   pitchCurve {editorCurveData {keys [
      {arriveTangent=-0.0013, interpMode=RCIM_Cubic, leaveTangent=-0.0013, time=0}
      <at 15 0.98 RCIM_Cubic>
      {arriveTangent=0.0013, interpMode=RCIM_Cubic, leaveTangent=0.0013, time=30}
     ]}}
   volumeCurve {editorCurveData {keys [<at 0 1>]}}
  }
 }
 {
  id=Enveloper_3
  children [Enveloper_0]
  properties {
   bLoop=true
   bLoopIndefinitely=true
   loopEnd=10
   pitchCurve {editorCurveData {keys [{time=0}]}}
   volumeCurve {editorCurveData {keys [
      {arriveTangent=-0.01, interpMode=RCIM_Cubic, time=0, value=1}
      <at 5 0.7 RCIM_Cubic>
      {arriveTangent=-0.03, interpMode=RCIM_Cubic, leaveTangent=-0.03, time=10, value=1}
      <at 15 0.4 RCIM_Cubic>
      <at 20 1 RCIM_Cubic>
     ]}}
  }
 }
 {id=Looping_0, children [Delay_0]}
 {id=Looping_1, children [Delay_1]}
 {id=Mixer_0, children [Looping_0,Looping_1,Enveloper_3], properties {inputVolume [1,1,1]}}
 {id=QualityLevel_0, children [Mixer_0,Mixer_0,Enveloper_3,Enveloper_3]}
 {id=Random_0, children [WavePlayer_2,WavePlayer_3,WavePlayer_4,WavePlayer_5,WavePlayer_6], properties {weights [1,1,1,1,1]}}
 {id=Random_1, children [WavePlayer_7, WavePlayer_8, WavePlayer_9, WavePlayer_10, WavePlayer_11], properties {weights [1,1,1,1,1]}}
 {
  id=WavePlayer_0
  properties {bLooping=true, soundWaveAssetPtr=&/Spot_FishTank_Bubble_Hum_Loop}
 }
 {id=WavePlayer_10, properties {soundWaveAssetPtr=&/Spot_FishTank_Bubble_Swell_04}}
 {id=WavePlayer_11, properties {soundWaveAssetPtr=&/Spot_FishTank_Bubble_Swell_05}}
 {id=WavePlayer_2, properties {soundWaveAssetPtr=&/Spot_FishTank_Bubble_Air_Rush_01}}
 {id=WavePlayer_3, properties {soundWaveAssetPtr=&/Spot_FishTank_Bubble_Air_Rush_02}}
 {id=WavePlayer_4, properties {soundWaveAssetPtr=&/Spot_FishTank_Bubble_Air_Rush_03}}
 {id=WavePlayer_5, properties {soundWaveAssetPtr=&/Spot_FishTank_Bubble_Air_Rush_04}}
 {id=WavePlayer_6, properties {soundWaveAssetPtr=&/Spot_FishTank_Bubble_Air_Rush_05}}
 {id=WavePlayer_7, properties {soundWaveAssetPtr=&/Spot_FishTank_Bubble_Swell_01}}
 {id=WavePlayer_8, properties {soundWaveAssetPtr=&/Spot_FishTank_Bubble_Swell_02}}
 {id=WavePlayer_9, properties {soundWaveAssetPtr=&/Spot_FishTank_Bubble_Swell_03}}
]
properties {
 bOverrideAttenuation=true
 bOverrideConcurrency=true
 maxDistance=1193.1147
 soundClassObject=&Fort_Ambient_Indoors_Athena
 subtitlePriority=5000
 volumeMultiplier=0.301
 attenuationOverrides {
  attenuationShape=Capsule
  bAttenuateWithLPF=true
  bEnableFocusInterpolation=true
  bEnableListenerFocus=true
  dBAttenuationAtMax=-40
  falloffDistance=649.4193
  focusAzimuth=15
  lPFFrequencyAtMax=750
  lPFRadiusMax=1000
  lPFRadiusMin=500
  nonFocusAzimuth=45
  nonFocusVolumeAttenuation=0.5
  occlusionInterpolationTime=2
  occlusionLowPassFilterFrequency=3000
  occlusionTraceChannel=ECC_GameTraceChannel11
  occlusionVolumeAttenuation=0.2
  stereoSpread=100
  attenuationShapeExtents=<543.6955 173.2973 131.6805>
 }
 concurrencyOverrides {bLimitToOwner=true, maxCount=2}
 soundSubmixSends [{
   minSendLevel=0.2
   soundSubmix=&Fortnite_Reverb_Default
   customSendLevelCurve {editorCurveData {keys [<at 0 0.2>,<at 1 1>]}}
  }]
}
