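# Sound cue for a looping fish tank hum, an ambient spot emitter with a box-shaped attenuation volume.
# A looping wave feeds two nested, indefinitely looping envelopers: the inner one gently drifts pitch, the outer modulates volume on a shorter cycle.
# Attenuation overrides add listener focus and occlusion filtering, and concurrency is limited per owner.
# It uses the indoor ambient sound class but sends to the exterior ambient source bus.
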SoundCue /Game/Sounds/Asteria/Environment_Actors/FishTank/Spot_Fish_Tank_Hum_Ext_Loop_Cue
root=Enveloper_3
nodes [
 {
  id=Enveloper_0
  children [WavePlayer_1]
  properties {
   bLoop=true
   bLoopIndefinitely=true
   loopEnd=30
   pitchCurve {editorCurveData {keys [
      {arriveTangent=-0.0013, interpMode=RCIM_Cubic, leaveTangent=-0.0013, time=0}
      <at 15 0.98 RCIM_Cubic>
      {arriveTangent=0.0013, interpMode=RCIM_Cubic, leaveTangent=0.0013, time=30}
     ]}}
   volumeCurve {editorCurveData {keys [<at 0 1>]}}
  }
 }
 {
  id=Enveloper_3
  children [Enveloper_0]
  properties {
   bLoop=true
   bLoopIndefinitely=true
   loopEnd=10
   pitchCurve {editorCurveData {keys [{time=0}]}}
   volumeCurve {editorCurveData {keys [
      {arriveTangent=-0.01, interpMode=RCIM_Cubic, time=0, value=1}
      <at 5 0.7 RCIM_Cubic>
      {arriveTangent=-0.03, interpMode=RCIM_Cubic, leaveTangent=-0.03, time=10, value=1}
      <at 15 0.4 RCIM_Cubic>
      <at 20 1 RCIM_Cubic>
     ]}}
  }
 }
 {id=WavePlayer_1, properties {bLooping=true, soundWaveAssetPtr=&/Spot_Fish_Tank_Hum_Ext_Loop}}
]
properties {
 bOverrideAttenuation=true
 bOverrideConcurrency=true
 maxDistance=808.2671
 soundClassObject=&Fort_Ambient_Indoors_Athena
 subtitlePriority=5000
 volumeMultiplier=0.551
 attenuationOverrides {
  attenuationShape=Box
  bEnableFocusInterpolation=true
  bEnableListenerFocus=true
  bEnableOcclusion=true
  dBAttenuationAtMax=-40
  falloffDistance=600
  focusAzimuth=15
  focusVolumeAttenuation=1.2
  nonFocusAzimuth=45
  nonFocusVolumeAttenuation=0.5
  occlusionInterpolationTime=2
  occlusionLowPassFilterFrequency=3000
  occlusionTraceChannel=ECC_GameTraceChannel11
  occlusionVolumeAttenuation=0.2
  stereoSpread=100
  attenuationShapeExtents=<208.2671 122.6473 131.6805>
 }
 busSends [{audioBus=null, minSendLevel=0.6, sendLevel=1, soundSourceBus=&AmbientAudio_Exterior_SourceBus}]
 concurrencyOverrides {bLimitToOwner=true, maxCount=2}
}
